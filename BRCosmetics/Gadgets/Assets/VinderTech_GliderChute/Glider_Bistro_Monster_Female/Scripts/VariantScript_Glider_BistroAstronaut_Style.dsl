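; Variant script for the Bistro Astronaut female glider cosmetic that decides which style variant the glider uses in a loadout.
; DetermineLoadoutVariantSelection checks whether the matching Bistro Astronaut character is in the loadout.
; If it is, the script returns that character's active Parts channel variant, so the glider style follows the outfit.
; Otherwise it returns the glider's own active Mesh channel variant.
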(fn DetermineLoadoutVariantSelection
  (params (Loadout :type Cosmetic%20Loadout%20Structure%20%28by%20ref%29))
  (if (Class|FortCosmeticFunctionLibrary|IsItemInLoadout Loadout "/BRCosmetics/Athena/Items/Cosmetics/Characters/CID_A_227_Athena_Commando_F_BistroAstronaut_JJLK5.CID_A_227_Athena_Commando_F_BistroAstronaut_JJLK5")
    (return (Class|FortCosmeticFunctionLibrary|GetActiveVariantForSoftItemAndChannel Loadout "/BRCosmetics/Athena/Items/Cosmetics/Characters/CID_A_227_Athena_Commando_F_BistroAstronaut_JJLK5.CID_A_227_Athena_Commando_F_BistroAstronaut_JJLK5" "(TagName=\"Cosmetics.Variant.Channel.Parts\")"))
    (else
      (return (Class|FortCosmeticFunctionLibrary|GetActiveVariantForSoftItemAndChannel Loadout "/BRCosmetics/Athena/Items/Cosmetics/Gliders/Glider_ID_319_BistroAstronautFemale_A4839.Glider_ID_319_BistroAstronautFemale_A4839" "(TagName=\"Cosmetics.Variant.Channel.Mesh\")")))))
